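# Unlit translucent material for the Byte Gear pickaxe's pixelated trail, enabled for Niagara ribbons, sprites and meshes.
# UVs are snapped to a tile grid with Ceil, so the trail mask and time-panned noise sample as blocky pixels, with mask U tiling from a dynamic parameter.
# The mask and noise product is tinted by particle color.
# Opacity math adds particle alpha, soft cell-edge masks from Frac of the grid, and a smoothstep fade along U.
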Material /BRCosmetics/Weapons/FORT_Melee/Pickaxe_ByteGear/FX/Materials/M_ByteGear_PixelatedTrail
graph {
 properties {
  bAutomaticallySetUsageInEditor=false
  bUsedWithInstancedStaticMeshes=true
  bUsedWithNiagaraMeshParticles=true
  bUsedWithNiagaraRibbons=true
  bUsedWithNiagaraSprites=true
  blendMode=BLEND_Translucent
  pixelDepthOffsetMode=PDOM_Legacy
  shadingModel=MSM_Unlit
  usedShadingModels=Unlit
 }
 nodes [
  n1 = AppendVector(A=ScalarParameter() {defaultValue=12, parameterName=SquareTileX}, B=ScalarParameter() {defaultValue=10, parameterName=SquareTileY})
  n2 = Multiply(A=TextureCoordinate() {vTiling=0.8}, B=@n1)
  n3 = Divide(A=Ceil(None=@n2), B=@n1)
  n4 = AppendVector(
   A=DynamicParameter() {paramNames [TileU, IntensityBias, MaxIntensity, NormalizedAge]}.TileU
   B=ScalarParameter() {defaultValue=1, parameterName=SmokePatternTile_V}
  )
  n5 = Multiply(A=@n3, B=@n4)
  n6 = TextureSampleParameter2D!(UVs=@n5, Level, Apply View MipBias) {
   mipValueMode=TMVM_MipLevel
   parameterName=TrailMask
   texture=/BRCosmetics/Weapons/FORT_Melee/Pickaxe_ByteGear/FX/Textures/T_Pickaxe_ByteGear_Trail_Mask
  }
  n7 = Add(A=Multiply(A=Time(), B=ScalarParameter() {defaultValue=1, parameterName=TimeMult}), B=ScalarParameter() {parameterName=TimeOffset})
  n8 = TextureSampleParameter2D(UVs=Panner(Coordinate=@n3, Time=@n7) {speedX=-1}) {
   mipValueMode=TMVM_MipLevel
   parameterName=PixelTextureA
   texture=/BRCosmetics/Weapons/FORT_Melee/Pickaxe_ByteGear/FX/Textures/T_Pickaxe_ByteGear_Trail_Noise
  }
  n9 = ScalarParameter() {defaultValue=1, parameterName=TrailBIntensity}
  MaterialExpressionConstantClamp_5 = Saturate(None=Multiply(A=@n6.R, B=LinearInterpolate(A=@n8.R, Alpha=@n9)))
  MaterialExpressionVertexColor_1 = ParticleColor()
  MaterialExpressionConstantClamp_4 = Saturate(None=Multiply(A=@MaterialExpressionConstantClamp_5, B=@MaterialExpressionVertexColor_1.RGB))
  MaterialFunctionCall(XY=@n5) {materialFunction=&Swizzle}
  n10 = Multiply(A=TextureCoordinate() {uTiling=-1, vTiling=-0.8}, B=@n1)
  n11 = Abs(None=ConstantBiasScale(None=Frac(None=@n10)))
  n12 = ScalarParameter() {defaultValue=0.97, parameterName=Size}
  n13 = OneMinus(None=Saturate(None=Divide(A=@n11, B=@n12)))
  n14 = ScalarParameter() {defaultValue=0.2, parameterName=Softness}
  n15 = MaterialFunctionCall(Float2=Saturate(None=Divide(A=@n13, B=@n14))) {materialFunction=&BreakOutFloat2Components}
  n16 = Multiply(A=@n15.R, B=@n15.R)
  n17 = Abs(None=ConstantBiasScale(None=Frac(None=@n2)))
  n18 = ScalarParameter() {defaultValue=0.97, parameterName=Size}
  n19 = OneMinus(None=Saturate(None=Divide(A=@n17, B=@n18)))
  n20 = ScalarParameter() {defaultValue=0.2, parameterName=Softness}
  n21 = MaterialFunctionCall(Float2=Saturate(None=Divide(A=@n19, B=@n20))) {materialFunction=&BreakOutFloat2Components}
  n22 = Multiply(A=@n21.R, B=@n21.R)
  n23 = TextureCoordinate()
  Multiply(
   A=Multiply(A=Multiply(A=@MaterialExpressionVertexColor_1.A, B=@MaterialExpressionConstantClamp_5), B=Multiply(A=@n16, B=@n22))
   B=SmoothStep(Value=ComponentMask(None=@n23) {r=true}) {constMax=0.4}
  )
  Panner(Coordinate=Multiply(A=@n3, B=@n4), Time=@n7) {speedX=-0.25}
 ]
}
