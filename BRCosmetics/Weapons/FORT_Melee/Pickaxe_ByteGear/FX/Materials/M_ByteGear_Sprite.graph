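# Translucent unlit Niagara sprite material for ByteGear pickaxe effects, giving an impact flash texture a pixelated, stepped look.
# Floor-based snapping of UVs and texel values quantizes the sample; particle alpha biases its blue channel, masked by green, for an erosion-style fade with optional depth fade.
# Particle color starts boosted and settles to normal over particle lifetime, except on mobile, which uses it unboosted, then feeds the shared emissive adjustment function.
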Material /BRCosmetics/Weapons/FORT_Melee/Pickaxe_ByteGear/FX/Materials/M_ByteGear_Sprite
graph {
 properties {bUsedWithNiagaraSprites=true, blendMode=BLEND_Translucent, pixelDepthOffsetMode=PDOM_Legacy, shadingModel=MSM_Unlit, usedShadingModels=Unlit}
 nodes [
  n1 = ParticleColor()
  n2 = Multiply(A=ScalarParameter() {defaultValue=4, parameterName="Particle color start  Multiply"}, B=@n1.RGB) {constA=4}
  n3 = ParticleRelativeTime()
  MaterialFunctionCall(Emissive=ShadingPathSwitch(Default=LinearInterpolate(A=@n2, B=@n1.RGB, Alpha=@n3) {constA=6}, Mobile=@n1.RGB)) {materialFunction=&MF_FX_EmissiveAdjustments}
  n4 = AppendVector(A=ScalarParameter() {defaultValue=128, parameterName=XPixels}, B=ScalarParameter() {defaultValue=64, parameterName=YPixels})
  n5 = Multiply(A=TextureCoordinate(), B=@n4)
  n6 = TextureSampleParameter2D!(UVs=Divide(A=Floor(None=@n5), B=@n4), Level, Apply View MipBias) {mipValueMode=TMVM_MipLevel, parameterName=ImpactTexture, texture=/Game/Effects/Fort_Effects/Textures/Explo/T_Impact_Flash_2x2_01}
  n7 = ScalarParameter() {defaultValue=64, parameterName=Values}
  n8 = Divide(A=Floor(None=Multiply(A=@n6.RGB, B=@n7)), B=@n7)
  n9 = Add(A=LinearInterpolate(Alpha=@n1.A) {constA=-0.6, constB=0.6}, B=ComponentMask(None=@n8) {b=true})
  n10 = ComponentMask(None=@n8) {g=true}
  n11 = Multiply(A=@n1.A, B=Clamp(None=Multiply(A=@n9, B=@n10)))
  n12 = ScalarParameter() {defaultValue=25, parameterName=DepthFade}
  Multiply(A=LinearInterpolate(A=@n11, B=DepthFade(Opacity=@n11, FadeDistance=@n12) {fadeDistanceDefault=25}, Alpha=ScalarParameter() {parameterName=UseDepthFade}))
 ]
}
