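# Masked material for Niagara mesh debris particles in the ByteGear pickaxe effects.
# Fakes sun lighting from SkyAtmosphere light direction and a normal map, summing a two-color diffuse ramp with a GGX highlight.
# Pixelated-UV noise modulates that lighting, a grid texture adds outlines, and particle color tints the result before emissive adjustment.
# Particle alpha drives temporal-AA dithering; a skylight reflection branch does not feed the emissive sum.
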Material /BRCosmetics/Weapons/FORT_Melee/Pickaxe_ByteGear/FX/Materials/M_Pickaxe_ByteGear_Debris
graph {
 properties {
  bEnableResponsiveAA=true
  bUsedWithNiagaraMeshParticles=true
  blendMode=BLEND_Masked
  pixelDepthOffsetMode=PDOM_Legacy
  translucencyLightingMode=TLM_Surface
  usedShadingModels=Default Lit
 }
 nodes [
  n1 = ParticleColor()
  n2 = MaterialFunctionCall(Alpha Threshold=@n1.A) {materialFunction=&DitherTemporalAA}
  If(A=@n1.A, "A > B"=Constant() {r=1}, "A == B"=@n2.Result, "A < B"=@n2.Result) {constB=0.99}
  n3 = Multiply(
   A=TextureSample(UVs=TextureCoordinate() {uTiling=1.5, vTiling=1.5}) {samplerType=SAMPLERTYPE_Grayscale, texture=/Game/Global/Base4Layers_2022/Textures/EnDGrid}.R
   B=ScalarParameter() {defaultValue=3, parameterName=OutlineIntensity}
  )
  n4 = SkyAtmosphereLightDirection()
  n5 = ScalarParameter() {defaultValue=0.2, parameterName=R}
  n6 = Transform(None=TextureSample() {samplerType=SAMPLERTYPE_Normal, texture=/Game/Effects/Fort_Effects/UI/Transition_IAS/Resources/T_ElimCubeNormal}.RGB)
  n7 = Multiply(
   A=Saturate(None=MaterialFunctionCall!(LightDirctionIn=@n4, CameraVectorIn, Roughness=@n5, NormalIn=@n6) {materialFunction=/Game/Effects/Niagara/MasterMaterials/Functions/Common/FXCommon_GGX}.Result)
   B=VectorParameter() {parameterName=R, defaultValue=<rgba 0.4 0.4 0.4 1>}.RGB
  )
  n8 = VectorParameter() {parameterName=Color2, defaultValue=<rgba 0.45 0.45 0.45 0.7>}
  n9 = ScalarParameter() {defaultValue=1, parameterName=CStrWithOutR}
  n10 = LinearInterpolate(
   A=VectorParameter() {parameterName=Color, defaultValue=<rgba 0.3 0.3 0.3 1>}.RGB
   B=Multiply(A=@n8.RGB, B=ShadingPathSwitch(Default=@n9, Deferred=@n9, Forward=@n9, Mobile=@n9))
   Alpha=Saturate(None=DotProduct(A=@n4, B=@n6))
  )
  n11 = Constant()
  n12 = TextureCoordinate()
  n13 = ScalarParameter() {defaultValue=9, parameterName=XPixels}
  n14 = Multiply(A=ComponentMask(None=@n12) {r=true}, B=@n13)
  n15 = ScalarParameter() {defaultValue=9, parameterName=YPixels}
  n16 = Multiply(A=ComponentMask(None=@n12) {g=true}, B=@n15)
  n17 = AppendVector(A=Divide(A=Floor(None=@n14), B=@n13), B=Divide(A=Floor(None=@n16), B=@n15))
  n18 = Multiply(
   A=Add(A=Add(A=@n7, B=@n10), B=ShadingPathSwitch(Default=@n11, Deferred=@n11, Forward=@n11, Mobile=@n11))
   B=Power(
    Base=TextureSample(UVs=@n17) {samplerType=SAMPLERTYPE_Grayscale, texture=&LowResBlurredNoise}.R
    Exp=ScalarParameter() {defaultValue=0.4, parameterName=TXPower}
   ) {constExponent=0.3}
  )
  MaterialFunctionCall(Emissive=Multiply(A=Add(A=@n3, B=@n18), B=@n1.RGB)) {materialFunction=&MF_FX_EmissiveAdjustments}
  n19 = ReflectionVectorWS(CustomWorldNormal=@n6)
  n20 = Saturate(None=Add(A=SkyLightEnvMapSample(Direction=@n19, Roughness=@n5)) {constB=0.01})
  Multiply(A=Multiply(A=@n8.RGB, B=Normalize(VectorInput=@n20)), B=ScalarParameter() {defaultValue=0.1, parameterName=A})
  ScalarParameter() {defaultValue=0.7, parameterName=CStrWithR}
 ]
}
