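# Material function from the ByteGear pickaxe cosmetic that turns a linear gradient into one reflected around a chosen center point.
# It saturates both inputs, divides the gradient by the center and the inverted gradient by the inverted center, then takes the minimum.
# The result is a triangular ramp peaking at the center and falling to zero at both ends.
# Inputs default to UV X and a debug time sine.
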MaterialFunction /BRCosmetics/Weapons/FORT_Melee/Pickaxe_ByteGear/Materials/MF_ByteGearGradientReflected
graph {
 properties {description=Linear gradient reflected around specified point, libraryCategoriesText [- TuataraVFX]}
 nodes [
  n1 = TextureCoordinate()
  n2 = Saturate(None=FunctionInput(Preview=ComponentMask(None=@n1) {r=true}) {bUsePreviewValueAsDefault=true, inputName=Gradient, inputType=FunctionInput_Scalar})
  n3 = FunctionInput(Preview=MaterialFunctionCall() {materialFunction=&DebugTimeSine}.Result) {
   bUsePreviewValueAsDefault=true
   desc=Point to reflect gradient around
   inputName=Center Point
   inputType=FunctionInput_Scalar
   sortPriority=1
  }
  n4 = Saturate(None=@n3)
  n5 = OneMinus(None=@n2)
  n6 = OneMinus(None=@n4)
  FunctionOutput(None=Min(A=Divide(A=@n2, B=@n4), B=Divide(A=@n5, B=@n6))) {description=Gradient reflected around centerpoint}
 ]
}
