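# Material function for the ByteGear pickaxe that planar-projects texture UVs onto the mesh from its pre-skinned position, so the projection ignores animation.
# Interpolated local X and Z are remapped between Start and End parameters to form the UV output, with the default Z range inverted.
# The Mask output uses Saturate, Frac and Ceil on each axis as a branchless bounds test, yielding one only strictly inside the projection rectangle.
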MaterialFunction /BRCosmetics/Weapons/FORT_Melee/Pickaxe_ByteGear/Materials/MF_Pickaxe_ByteGear_TextureProj
graph {nodes [
  n1 = VertexInterpolator(VS=PreSkinnedPosition())
  n2 = ScalarParameter() {defaultValue=-30.85, parameterName=Texture Projection X Start}
  n3 = Divide(
   A=Subtract(A=ComponentMask(None=@n1.PS) {r=true}, B=@n2)
   B=Subtract(A=ScalarParameter() {defaultValue=-11.25, parameterName=Texture Projection X End}, B=@n2)
  )
  n4 = ScalarParameter() {defaultValue=44.9, parameterName=Texture Projection Y Start}
  n5 = Divide(
   A=Subtract(A=ComponentMask(None=@n1.PS) {b=true}, B=@n4)
   B=Subtract(A=ScalarParameter() {defaultValue=7.85, parameterName=Texture Projection Y End}, B=@n4)
  )
  FunctionOutput(None=AppendVector(A=@n3, B=@n5)) {outputName=UV}
  n6 = Frac(None=Saturate(None=@n3))
  n7 = Frac(None=Saturate(None=@n5))
  FunctionOutput(None=Multiply(A=Ceil(None=@n6), B=Ceil(None=@n7))) {outputName=Mask, sortPriority=1}
 ]}
