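# Material function for the ByteGear pickaxe, layering animated holographic effects onto input material attributes, with output through opaque fake translucency.
# A projection subfunction and blade mask texture drive a timed erosion sheen, sine shimmer and cell mask.
# Scanlines, flicker, scrolling noise, curve-atlas colors and custom-depth comparison shape emissive and opacity.
# It also holds vertex-color-driven world position offset wiggle, a DeRez transition and apparently unconnected leftover branches.
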MaterialFunction /BRCosmetics/Weapons/FORT_Melee/Pickaxe_ByteGear/Materials/MF_Pickaxe_ByteGear
graph {nodes [
  n1 = NamedRerouteUsage()
  n2 = GetMaterialAttributes(None=@n1) {attributeGetTypes [#1, #2]}
  n3 = Multiply(
   A=Fresnel() {baseReflectFraction=0}
   B=VectorParameter() {group=00_SuperLevels_T3, parameterName=SL_FresnelColor, defaultValue=<rgba 0.6094 3 0.6925 1>}.RGB
  )
  n4 = VectorParameter() {group=00_SuperLevels_T3, parameterName=SL_T3_SecondaryTintColor, defaultValue=<rgba 0.0119 0.026 0.0117 1>}
  n5 = VectorParameter() {group=00_SuperLevels_T3, parameterName=SL_T3_SecondaryPatternColor, defaultValue=<rgba 0.2127 0.5938 0.3055 1>}
  n6 = TextureSample(UVs=Multiply(A=TextureCoordinate(), B=ScalarParameter() {defaultValue=1, group=00_SuperLevels_T3, parameterName=SL_T3_SecondaryUVScale})) {
   samplerSource=SSM_Wrap_WorldGroupSettings
   samplerType=SAMPLERTYPE_LinearColor
   texture=/BRCosmetics/Weapons/FORT_Melee/Pickaxe_ByteGear/Textures/T_ByteGear_Pickaxe_FX
  }
  n7 = Subtract(A=Subtract(A=Time() {bOverride_Period=true, period=4}, B=@n6.G)) {constB=2.8}
  n8 = Frac(None=Multiply(A=ScalarParameter() {defaultValue=1, group=00_SuperLevels_T3, parameterName=SL_T3_SecondaryPatternSpeed}, B=Saturate(None=@n7)))
  n9 = SphereMask(A=@n8, B=Constant() {r=0.5}) {attenuationRadius=0.125, hardnessPercent=0.1}
  n10 = Constant() {r=0.5}
  n11 = LinearInterpolate(A=@n4.RGB, B=@n5.RGB, Alpha=Multiply(A=@n9, B=SphereMask(A=@n8, B=@n10) {attenuationRadius=0.1, hardnessPercent=0.5}))
  n12 = Add(A=Multiply(A=NamedRerouteUsage()) {constB=0.05}, B=Multiply(A=Add(A=@n3, B=@n11)) {constB=24})
  Add(A=@n2.EmissiveColor, B=@n12)
  Constant() {r=1}
  Constant() {r=3}
  FunctionInput() {inputName=TechArtMask, sortPriority=1}
  FunctionOutput(None=SetMaterialAttributes!(
    MaterialAttributes=MaterialFunctionCall!(Material=@n1, Alpha) {materialFunction=/Game/Global/MaterialFunctions/MF_OpaqueFakeTranslucency}.Result
    EmissiveColor=NamedRerouteUsage()
    AmbientOcclusion=Multiply(A=@n2.AmbientOcclusion)
   ) {attributeSetTypes [#1, #2]})
  n13 = Multiply(A=NamedRerouteUsage(), B=VectorParameter() {parameterName=Sheen Color, defaultValue=<rgba 0 6 5.0289 0>}.RGB)
  n14 = Multiply(A=NamedRerouteUsage(), B=VectorParameter() {parameterName=Shimmer Color, defaultValue=<rgba 0.65 1.5 0 0>}.RGB)
  n15 = ScalarParameter() {defaultValue=3, parameterName=Shimmer Duration}
  n16 = ScalarParameter() {defaultValue=1, parameterName=Sheen Duration}
  n17 = Fmod(A=Time() {bOverride_Period=true, period=4}, B=Add(A=@n15, B=@n16))
  n18 = Subtract(A=@n17, B=@n16)
  LinearInterpolate(A=@n13, B=@n14, Alpha=Saturate(None=Sign(None=@n18)))
  n19 = ScalarParameter() {defaultValue=-2, parameterName=DepthBias}
  MaterialFunctionCall!(Opacity=@n19) {materialFunction=/Game/Global/Materials/MaterialGym/Materials/SortedOutline/MF_SortTranslucency}
  n20 = GetMaterialAttributes(None=SetMaterialAttributes!(MaterialAttributes=NamedRerouteUsage(), EmissiveColor=NamedRerouteUsage(), Opacity=NamedRerouteUsage()) {attributeSetTypes [#1, #3]}) {attributeGetTypes [#3, #1]}
  n21 = Subtract(A=WorldPosition().XYZ, B=CameraPositionWS())
  n22 = ComponentMask(None=Transform(None=Add(A=@n21, B=@n19)) {transformSourceType=TRANSFORMSOURCE_World, transformType=TRANSFORM_View}) {b=true}
  n23 = SceneTexture() {sceneTextureId=PPI_CustomDepth}
  n24 = Constant() {r=0.1}
  n25 = SetMaterialAttributes!(
   MaterialAttributes=@n20.MaterialAttributes
   Opacity=If(A=Multiply(A=@n22) {constB=0.9999}, B=ComponentMask(None=@n23.Color) {r=true}, "A > B"=Constant(), "A == B"=@n24, "A < B"=@n24)
   EmissiveColor=Multiply(A=NamedRerouteUsage(), B=ScalarParameter() {defaultValue=1, parameterName=EmissiveMultiplier})
  ) {attributeSetTypes [#3, #1]}
  MaterialFunctionCall(In=@n25) {materialFunction=&MF_TechArtOverrides_accessories}
  n26 = Add(A=NamedRerouteUsage(), B=Multiply(A=Time(), B=ScalarParameter() {defaultValue=0.1, parameterName=Shimmer Speed}))
  n27 = Saturate(None=Divide(A=Frac(None=@n26), B=ScalarParameter() {defaultValue=0.3, parameterName=Shimmer Mask Softness}))
  n28 = Multiply(A=Sine(None=Divide(A=@n27))) {constB=0.5}
  n29 = ScalarParameter() {defaultValue=0.67, parameterName=Shimmer Speed}
  n30 = Add(A=NamedRerouteUsage(), B=Multiply(A=Time(), B=Subtract(A=@n29) {constB=0.05}))
  n31 = ScalarParameter() {defaultValue=0.3, parameterName=Shimmer Mask Softness}
  n32 = Saturate(None=Divide(A=Frac(None=@n30), B=Add(A=@n31) {constB=0.2}))
  n33 = Multiply(A=Sine(None=Divide(A=@n32))) {constB=0.2}
  n34 = ScalarParameter() {defaultValue=0.67, parameterName=Shimmer Speed}
  n35 = Add(A=NamedRerouteUsage(), B=Multiply(A=Time(), B=Subtract(A=@n34) {constB=0.075}))
  n36 = ScalarParameter() {defaultValue=0.3, parameterName=Shimmer Mask Softness}
  n37 = Saturate(None=Divide(A=Frac(None=@n35), B=Add(A=@n36) {constB=0.5}))
  n38 = Multiply(A=Sine(None=Divide(A=@n37))) {constB=0.15}
  Max(A=Max(A=@n28, B=@n33), B=@n38)
  Multiply(B=Fresnel())
  Multiply(A=@n3, B=@n11)
  Multiply(A=@n2.EmissiveColor, B=@n12)
  n39 = Divide(A=Saturate(None=Divide(A=@n18, B=@n15)))
  Multiply(A=Power(Base=Sine(None=@n39)), B=NamedRerouteUsage())
  n40 = LinearInterpolate(Alpha=@n8) {constA=2, constB=0.05}
  Multiply(
   A=@n9
   B=SphereMask(A=SphereMask(A=@n40, B=@n6.R) {attenuationRadius=0.125, hardnessPercent=0}, B=Constant() {r=0.9}) {attenuationRadius=0.5, hardnessPercent=0}
  )
  n41 = ComponentMask(None=TransformPosition(None=WorldPosition().XYZ) {transformSourceType=TRANSFORMPOSSOURCE_World}) {b=true}
  n42 = ComponentMask(None=TextureCoordinate()) {g=true}
  n43 = ScalarParameter() {parameterName=MaskPoison}
  n44 = AppendVector(
   A=ComponentMask(None=TextureCoordinate() {vTiling=0.5}) {r=true}
   B=Divide(A=LinearInterpolate(A=@n41, B=@n42, Alpha=@n43), B=ScalarParameter() {defaultValue=128, parameterName=FlickerScale}) {constB=32}
  )
  n45 = Ceil(None=Multiply(A=Time()) {constB=3})
  n46 = VectorParameter() {parameterName=Param_1, defaultValue=<rgba 0.1 -0.1 0 1>}
  n47 = Divide(A=@n46.R) {constB=1000}
  n48 = TextureSample(UVs=Panner(Coordinate=@n44, Time=@n45, Speed=AppendVector(A=@n46.R, B=@n47))) {samplerType=SAMPLERTYPE_LinearColor, texture=/Game/UI/Foundation/Textures/Pattern/T_CubePattern}
  n49 = VectorParameter() {defaultValue=<rgba 0.3 -0.1 0 1>}
  n50 = Divide(A=@n49.G) {constB=1000}
  n51 = TextureSample(UVs=Panner(Coordinate=@n44, Time=@n45, Speed=AppendVector(A=@n50, B=@n49.G))) {samplerType=SAMPLERTYPE_LinearColor, texture=&T_CubePattern}
  n52 = OneMinus(None=Round(None=Subtract(A=@n48.R, B=@n51.G)))
  NamedRerouteDeclaration(None=LinearInterpolate(A=Constant() {r=1}, B=Clamp(None=@n52), Alpha=ScalarParameter() {parameterName=Flicker})) {name=ScanLinesVerB, nodeColor=<rgba 0 1 0.5781 1>}
  n53 = GetMaterialAttributes(None=FunctionInput() {inputType=FunctionInput_MaterialAttributes}) {attributeGetTypes [#4, #5]}
  NamedRerouteDeclaration(None=@n53.MaterialAttributes) {name=IN_BaseMaterial, nodeColor=<rgba 1 0.5859 0 1>}
  n54 = VertexColor()
  n55 = Multiply(A=Time(), B=ScalarParameter() {defaultValue=0.05, parameterName=WPO_Speed})
  n56 = TextureSample(UVs=Divide(A=Add(A=@n54.R, B=@n55), B=Constant2Vector() {g=1, r=64})) {samplerType=SAMPLERTYPE_LinearColor, texture=/Game/Effects/Fort_Effects/Textures/Tile/LinerNoise64x64}
  NamedRerouteDeclaration(None=@n56.G) {name=NoiseMapG, nodeColor=<rgba 0.1207 1 0.1311 1>}
  NamedRerouteDeclaration(None=@n56.B) {name=NoiseMapB, nodeColor=<rgba 0 0.2266 1 1>}
  n57 = VectorParameter() {parameterName=UV2, defaultValue=<rgba 0.5 1.5 0.1 0.206>}
  n58 = ComponentMask(None=@n57.RGB) {g=true, r=true}
  n59 = LinearInterpolate(Alpha=VertexColor().G) {constA=1, constB=0.309}
  n60 = Panner(Coordinate=Multiply(A=TextureCoordinate() {coordinateIndex=1}, B=Multiply(A=@n58, B=@n59)), Speed=AppendVector(A=@n57.B, B=@n57.B))
  n61 = Multiply(A=ComponentMask(None=NamedRerouteUsage()) {g=true, r=true}, B=ScalarParameter() {defaultValue=0.02, parameterName=NormalAffectMap})
  NamedRerouteDeclaration(None=TextureSample(UVs=Add(A=@n60, B=@n61)) {texture=&T_NoisePack_01}.R) {name=NoisePackMap_Red, nodeColor=<rgba 1 0.1389 0.1198 1>}
  n62 = TextureCoordinate() {coordinateIndex=1}
  n63 = ComponentMask(None=@n62) {r=true}
  n64 = AppendVector(A=Multiply(A=@n63, B=@n63), B=ComponentMask(None=@n62) {g=true})
  n65 = TextureSample(UVs=Panner(Coordinate=MaterialFunctionCall(XY=@n64) {materialFunction=&Swizzle}.YX) {speedX=1}) {texture=/Game/Effects/Fort_Effects/Textures/T_FX_Circuitboard2}
  NamedRerouteDeclaration(None=@n65.G) {name=CircuitBoardG, nodeColor=<rgba 1 0 0.8672 1>}
  NamedRerouteDeclaration(None=@n65.B) {name=CircuitBoardB, nodeColor=<rgba 0 1 0.8125 1>}
  n66 = Time()
  n67 = Multiply(A=VertexColor().R) {constB=2}
  n68 = Sine(None=Multiply(A=Add(A=@n66, B=@n67), B=ScalarParameter() {defaultValue=0.15, parameterName=WiggleFreq}))
  n69 = ScalarParameter() {defaultValue=2, parameterName=WiggleDistance}
  n70 = Multiply(A=Transform(None=Constant3Vector() {constant=<rgba 0 0 1 0>}) {transformSourceType=TRANSFORMSOURCE_Local}, B=@n69)
  NamedRerouteDeclaration(None=Clamp(None=Multiply(A=@n68, B=@n70), Max=@n69) {maxDefault=5}) {name=WiggleAmount, nodeColor=<rgba 1 0.4922 0 1>}
  n71 = Power(Base=NamedRerouteUsage(), Exp=ScalarParameter() {defaultValue=4, parameterName=WPO_MotionExp})
  n72 = Subtract(
   A=MaterialFunctionCall(X=ComponentMask(None=@n71) {r=true}, Y=NamedRerouteUsage(), Z=NamedRerouteUsage()) {materialFunction=&MakeFloat3}.Result
   B=Constant3Vector() {constant=<rgba 0 0.5 0.5 0>}
  )
  n73 = VectorParameter() {parameterName=WPO_Rand_Size, defaultValue=<rgba 6 0.5 0.5 1.25>}
  NamedRerouteDeclaration(None=Transform(None=Multiply(A=@n72, B=@n73.RGB)) {transformSourceType=TRANSFORMSOURCE_Local}) {name=WorldSpaceNoise, nodeColor=<rgba 0.1719 0 1 1>}
  n74 = VertexColor()
  n75 = Multiply(A=@n71, B=VectorParameter() {parameterName=WPO_Rand_Size, defaultValue=<rgba 6 0.5 0.5 1.25>}.A) {constB=2}
  n76 = TransformPosition(None=WorldPosition().XYZ) {transformSourceType=TRANSFORMPOSSOURCE_World}
  n77 = Multiply(A=NamedRerouteUsage(), B=Normalize(VectorInput=ComponentMask(None=@n76) {r=true}))
  n78 = Add(A=NamedRerouteUsage(), B=Saturate(None=If(A=@n74.G, "A > B"=@n75, "A < B"=@n77) {constB=0.5}))
  NamedRerouteDeclaration(None=SetMaterialAttributes!(MaterialAttributes, WorldPositionOffset=NamedRerouteDeclaration(None=@n78) {name=WPOWiggle, nodeColor=<rgba 0 1 0.4844 1>}) {attributeSetTypes [#6]}) {name=MaterialAttributeWPO, nodeColor=<rgba 0.1484 1 0 1>}
  n79 = TextureSampleParameter2D() {
   group=Emissive Mask
   parameterName=Emissive Mask Texture
   sortPriority=1
   texture=/BRCosmetics/Accessories/FORT_Backpacks/F_MED_Graffiti_Future_Backpack/Skins/Orange/Textures/T_F_MED_GraffitiFuture_Orange_backpack_FX
  }
  NamedRerouteDeclaration(None=@n79.G) {name=EmissiveMaskTextureG, nodeColor=<rgba 0.057 1 0.1229 1>}
  n80 = VectorParameter() {
   group=Emissive Mask
   parameterName=Emissive Mask Color/Alpha
   sortPriority=3
   channelNames {a=Alpha Strength, b="", g="", r=""}
   defaultValue=<rgba 0 0 0 1>
  }
  NamedRerouteDeclaration(None=@n80) {name=EmissiveMaskColorRGB, nodeColor=<rgba 0 0.7891 1 1>}
  NamedRerouteDeclaration(None=@n53.Normal) {name=BaseMaterial_Normal, nodeColor=<rgba 0.2656 0 1 1>}
  NamedRerouteDeclaration(None=@n80.Alpha Strength) {name=EmissiveMaskAlpha, nodeColor=<rgba 1 0.8906 0 1>}
  NamedRerouteDeclaration(None=ScalarParameter() {group=Emissive Mask, parameterName=Use Emissive Mask, sortPriority=0}) {name=UseEmissiveMaskAmount, nodeColor=<rgba 0.5703 0 1 1>}
  n81 = CameraVectorWS()
  n82 = Transform(None=LinearInterpolate(
    A=Constant3Vector() {constant=<rgba 0 0 1 0>}
    B=NamedRerouteUsage()
    Alpha=ScalarParameter() {defaultValue=0.25, parameterName=NormalStrength}
   ))
  n83 = CurveAtlasRowParameter(CurveTime=Saturate(None=DotProduct(A=@n81, B=@n82))) {
   atlas=&AthenaCurveAtlas01
   curve=/Game/Accessories/FORT_Backpacks/M_MED_StreetDemon_Backpack/FX/Curve_Color
   defaultValue=81
   parameterName=ColorCurveFromNormalMap
  }
  NamedRerouteDeclaration(None=MaterialFunctionCall(Hue Shift Percentage=ScalarParameter() {parameterName=HueShiftColorCurve}, Texture=@n83) {materialFunction=&HueShift}.Result) {name=ColorCurveFromNormalMapRGB, nodeColor=<rgba 0 1 0.25 1>}
  n84 = ComponentMask(None=TextureCoordinate() {coordinateIndex=1}) {g=true}
  n85 = CurveAtlasRowParameter(CurveTime=Power(Base=OneMinus(None=@n84))) {
   atlas=&AthenaCurveAtlas01
   curve=/Game/Athena/LinearColorCurves/HDRFlamesCurve
   defaultValue=60
   parameterName=ColorCurveHDRFlamesCurve
  }
  NamedRerouteDeclaration(None=@n85) {name=ColorCurveHDRFlamesCurve, nodeColor=<rgba 1 0 0.0703 1>}
  NamedRerouteDeclaration(None=@n83.A) {name=ColorCurveFromNormalMapA, nodeColor=<rgba 0 0.3906 1 1>}
  n86 = NamedRerouteUsage()
  n87 = LinearInterpolate(
   A=NamedRerouteUsage()
   B=LinearInterpolate(A=NamedRerouteUsage(), B=NamedRerouteUsage(), Alpha=NamedRerouteUsage())
   Alpha=NamedRerouteUsage()
  )
  NamedRerouteDeclaration(None=LinearInterpolate(A=NamedRerouteUsage(), B=Multiply(A=@n86, B=@n87), Alpha=ScalarParameter() {parameterName=Use_WorldPositionMask})) {name=EmissiveColorCalculation, nodeColor=<rgba 0.7109 1 0 1>}
  n88 = LinearInterpolate(
   A=NamedRerouteUsage()
   B=LinearInterpolate(A=NamedRerouteUsage(), B=NamedRerouteUsage(), Alpha=Constant() {r=0.3})
   Alpha=ScalarParameter() {parameterName=MaskPoison}
  )
  NamedRerouteDeclaration(None=Multiply(A=NamedRerouteUsage(), B=LinearInterpolate(Alpha=@n88) {constA=0.25})) {name=OpacityCalculation, nodeColor=<rgba 0.9688 0 1 1>}
  NamedRerouteDeclaration(None=@n20.Opacity) {name=UpdatedOpacity, nodeColor=<rgba 0 1 0.6484 1>}
  n89 = Constant() {r=16}
  n90 = Constant() {r=-0.35}
  n91 = Constant() {r=32}
  n92 = Constant() {r=0.61}
  n93 = Constant() {r=2.2235}
  n94 = ComponentMask(None=MaterialFunctionCall() {materialFunction=&BoundingBoxBased_0-1_UVW}.RGB) {b=true, r=true}
  n95 = Add(A=Multiply(A=MaterialFunctionCall!(ScanLine_Tiling_1=@n89, Scanline_PannerSpeed_1=@n90, ScanLine_Tiling=@n91, Scanline_PannerSpeed=@n92, Scanline_Rotation=@n93, UVs=@n94) {desc=ScanLine, materialFunction=/Game/Global/Materials/MaterialGym/Materials/HolographicTexture/MF_HolographicTexture_ScanLines}.Result) {constB=2}) {constB=0.5}
  n96 = MaterialFunctionCall() {materialFunction=&BoundingBoxBased_0-1_UVW}
  n97 = Multiply(A=MaterialFunctionCall(
    ScanLine_Tiling_1=Constant() {r=8}
    Scanline_PannerSpeed_1=Constant() {r=0.5}
    ScanLine_Tiling=Constant() {r=17}
    Scanline_PannerSpeed=Constant() {r=-0.15}
    Scanline_Rotation=Constant() {r=2.2235}
    UVs=ComponentMask(None=@n96.RGB) {b=true, r=true}
   ) {desc=ScanLine, materialFunction=&MF_HolographicTexture_ScanLines}.Result) {constB=2}
  n98 = Multiply(A=Multiply(A=NamedRerouteUsage(), B=Saturate(None=@n95)), B=Saturate(None=Add(A=@n97) {constB=0.5})) {constA=1}
  n99 = NamedRerouteUsage()
  n100 = NamedRerouteUsage()
  NamedRerouteDeclaration(None=LinearInterpolate(A=@n98, B=LinearInterpolate(A=@n98, B=@n99, Alpha=@n100), Alpha=NamedRerouteUsage())) {name=OpacityWithScanLines, nodeColor=<rgba 1 0.3281 0 1>}
  NamedRerouteDeclaration(None=@n20.EmissiveColor) {name=UpdatedEmissiveColor, nodeColor=<rgba 0.8794 1 0.0276 1>}
  NamedRerouteDeclaration(None=@n53.BaseColor) {name=BaseMaterial_Color, nodeColor=<rgba 0.0547 1 0 1>}
  n101 = Constant() {r=1}
  n102 = Constant() {r=0.25}
  n103 = Constant() {r=24}
  n104 = Constant() {r=-0.16}
  n105 = Constant() {r=2.2235}
  n106 = ComponentMask(None=MaterialFunctionCall() {materialFunction=&BoundingBoxBased_0-1_UVW}.RGB) {b=true, r=true}
  n107 = Multiply(A=OneMinus(None=MaterialFunctionCall(ScanLine_Tiling_1=@n101, Scanline_PannerSpeed_1=@n102, ScanLine_Tiling=@n103, Scanline_PannerSpeed=@n104, Scanline_Rotation=@n105, UVs=@n106) {desc=ScanLine, materialFunction=&MF_HolographicTexture_ScanLines}.Result)) {constB=0.25}
  n108 = Multiply(A=NamedRerouteUsage(), B=Saturate(None=Add(A=@n107) {constB=0.1}))
  n109 = NamedRerouteUsage()
  n110 = NamedRerouteUsage()
  NamedRerouteDeclaration(None=LinearInterpolate(A=@n108, B=LinearInterpolate(A=@n108, B=@n109, Alpha=@n110), Alpha=NamedRerouteUsage())) {name=EmissiveColorWithScanlines, nodeColor=<rgba 0.3125 1 0 1>}
  NamedRerouteDeclaration(None=MaterialFunctionCall!(Fresnel=ScalarParameter() {defaultValue=1, parameterName=Fresnel}, Material=@n25) {materialFunction=/Game/Packages/Fortress_SharedMaterials/Base_Material_Functions/CharacterDeRez_Transition}.Result) {name=FinalMaterialAttributes, nodeColor=<rgba 0.4072 0.0343 0.0056 1>}
  n111 = PixelDepth()
  n112 = ScalarParameter() {defaultValue=-2, parameterName=DepthBias}
  n113 = SceneTexture() {sceneTextureId=PPI_CustomDepth}
  NamedRerouteDeclaration(None=If(A=Add(A=@n111.R, B=@n112), B=ComponentMask(None=@n113.Color) {r=true}, "A > B"=Constant(), "A == B"=@n24, "A < B"=@n24)) {name=depthopacity, nodeColor=<rgba 0 1 0.6953 1>}
  n114 = MaterialFunctionCall() {materialFunction=&/MF_Pickaxe_ByteGear_TextureProj}
  NamedRerouteDeclaration(None=@n114.Mask) {name=Blade Mask, nodeColor=<rgba 0.2891 1 0 1>}
  n115 = Add(A=@n28, B=@n33)
  NamedRerouteDeclaration(None=Multiply(A=Add(A=@n115, B=@n38)) {constB=0.333}) {name=Shimmer, nodeColor=<rgba 0.4297 0 1 1>}
  n116 = ScalarParameter() {defaultValue=0.35, parameterName=Sheen Angle, sliderMax=1}
  n117 = AppendVector(A=Cosine(None=@n116), B=Sine(None=@n116))
  n118 = TextureSample(UVs=@n114.UV) {samplerType=SAMPLERTYPE_LinearColor, texture=/BRCosmetics/Weapons/FORT_Melee/Pickaxe_ByteGear/Textures/T_Pickaxe_ByteGear_BladeMasks_FX}
  n119 = ConstantBiasScale(None=AppendVector(A=@n118.R, B=@n118.R)) {bias=-0.5, scale=2}
  n120 = Saturate(None=Subtract(A=@n17) {constB=3})
  n121 = MaterialFunctionCall!(
   Gradient=ConstantBiasScale(None=DotProduct(A=@n117, B=@n119))
   Erosion Amount=Frac(None=Divide(A=@n120, B=@n16))
   Softness=ScalarParameter() {defaultValue=0.6, parameterName=Sheen Size}
  ) {materialFunction=&/MF_ByteGearErosion}
  n122 = ScalarParameter() {defaultValue=0.7, parameterName=Sheen Gradient Center Point}
  NamedRerouteDeclaration(None=Power(
    Base=MaterialFunctionCall!(Gradient=@n121.Increased Range, Center Point=@n122) {materialFunction=&/MF_ByteGearGradientReflected}.Result
    Exp=ScalarParameter() {defaultValue=3, parameterName=Sheen Exponent}
   )) {name=Sheen, nodeColor=<rgba 0 0.9062 1 1>}
  n123 = Add(A=OneMinus(None=@n118.R)) {constB=0.2}
  n124 = ScalarParameter() {defaultValue=0.5, parameterName=Cells Mask Exponent}
  NamedRerouteDeclaration(None=Multiply(A=Saturate(None=@n123), B=Power(Base=@n118.A, Exp=@n124))) {name=Cells Mask, nodeColor=<rgba 0.1016 1 0 1>}
  n125 = NamedRerouteUsage()
  n126 = Multiply(A=@n13) {constB=0.8}
  n127 = StaticSwitchParameter(
   True=Multiply(A=VectorParameter() {parameterName=Fade In Color, defaultValue=<rgba 6 5.1 0 0>}.RGB, B=NamedRerouteUsage())
   False=Add(A=Multiply(A=@n125) {constB=0.01}, B=Add(A=@n126, B=@n14))
  ) {parameterName=Is Fade In Mesh}
  n128 = NamedRerouteUsage()
  NamedRerouteDeclaration(None=Multiply(A=Multiply(A=@n127, B=@n128), B=NamedRerouteUsage())) {name=BrunoSheen, nodeColor=<rgba 1 0 0.1172 1>}
  NamedRerouteDeclaration(None=@n118.B) {name=BlueChannelTexture, nodeColor=<rgba 0 1 0.5781 1>}
  n129 = MaterialFunctionCall() {materialFunction=&BoundingBoxBased_0-1_UVW}
  NamedRerouteDeclaration(None=Multiply(A=ComponentMask(None=@n129.RGB) {b=true, r=true})) {name=ScanlineUVs, nodeColor=<rgba 0 1 0.8125 1>}
  n130 = WorldPosition()
  n131 = LinearInterpolate(
   A=ComponentMask(None=TransformPosition(None=@n130.XYZ) {transformSourceType=TRANSFORMPOSSOURCE_World}) {b=true}
   B=ComponentMask(None=TextureCoordinate()) {g=true}
   Alpha=ScalarParameter() {parameterName=MaskPoison}
  )
  n132 = VectorParameter() {parameterName=LineSpeed, defaultValue=<rgba 0.1 -0.1 0 0>}
  n133 = Panner(
   Coordinate=Divide(A=@n131, B=ScalarParameter() {defaultValue=12, parameterName=LineScale1}) {constB=32}
   Speed=AppendVector(A=@n132.R, B=Divide(A=@n132.R) {constB=1000})
  )
  NamedRerouteDeclaration(None=LinearInterpolate(Alpha=TextureSample(UVs=@n133) {samplerType=SAMPLERTYPE_Normal, texture=/Game/Global/Textures/Default/Noise/64x64Noise}.R) {constA=1, constAlpha=0.1, constB=0.55}) {name=AnimatedLinesA, nodeColor=<rgba 1 0.2578 0 1>}
  n134 = Panner(
   Coordinate=Divide(A=@n131, B=ScalarParameter() {defaultValue=128, parameterName=LineScale2}) {constB=32}
   Speed=AppendVector(A=@n132.G, B=Divide(A=@n132.G) {constB=100000})
  )
  n135 = Multiply(A=Subtract(
    A=TextureSample(UVs=@n134) {samplerType=SAMPLERTYPE_Normal, texture=&64x64Noise}.G
    B=ScalarParameter() {defaultValue=0.9, parameterName=LineSub}
   ) {constB=0.99}) {constB=15}
  NamedRerouteDeclaration(None=LinearInterpolate(Alpha=Saturate(None=@n135)) {constA=0.5, constB=2}) {name=AnimatedLinesB, nodeColor=<rgba 0 0.0625 1 1>}
  n136 = ScalarParameter() {defaultValue=1, group=FakeGlass, parameterName=FlipNormals, sortPriority=0}
  n137 = Multiply(A=LinearInterpolate(Alpha=Saturate(None=@n136)) {constA=1, constB=-1}, B=TwoSidedSign())
  NamedRerouteDeclaration(None=ShadowReplace(Default=Saturate(None=@n137), Shadow=Constant() {r=1})) {name=OpacityMask, nodeColor=<rgba 1 0 0.375 1>}
  n138 = VectorParameter() {parameterName=UV, defaultValue=<rgba 0.5 0.5 0.05 0.0679>}
  n139 = Multiply(A=TextureCoordinate(), B=ComponentMask(None=@n138.RGB) {g=true, r=true})
  n140 = AppendVector(A=@n138.B, B=@n138.B)
  NamedRerouteDeclaration(None=TextureSampleParameter2D(UVs=Panner(Coordinate=@n139, Speed=@n140)) {parameterName=Param, samplerType=SAMPLERTYPE_Normal, texture=/Game/Global/Textures/Default/Tileable_Patterns/Cloud_normals}.RGB) {name=NormalMap, nodeColor=<rgba 0.7344 0 1 1>}
  NamedRerouteDeclaration(None=@n56.RGB) {name=NoiseMapRGB, nodeColor=<rgba 1 0.8883 0.8802 1>}
  NamedRerouteUsage()
  NamedRerouteUsage()
  NamedRerouteUsage()
  NamedRerouteUsage()
  SetMaterialAttributes!(
   MaterialAttributes
   BaseColor=NamedRerouteUsage()
   Roughness=Constant() {r=0.5}
   Metallic=Constant()
   Specular=Constant() {r=0.5}
   EmissiveColor=Constant()
   Normal=NamedRerouteUsage()
  ) {attributeSetTypes [
    #5
    #7
    #8
    #9
    #1
    #4
   ]}
 ]}
