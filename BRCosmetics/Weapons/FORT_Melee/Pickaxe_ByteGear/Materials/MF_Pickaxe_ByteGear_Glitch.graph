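# Material function for the ByteGear pickaxe that layers a glitchy hologram over incoming material attributes, also outputting emissive and opacity.
# A screen-glitched texture or base color, modulated by scanlines, is blended by a selectable mask into emissive and opacity, optionally blackening base color.
# A sine-timed, randomly flickering glitch mask and an FX-texture cube-outline term are also built.
# An optional static switch enables vertex jitter from panned noise and remapped normals.
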MaterialFunction /BRCosmetics/Weapons/FORT_Melee/Pickaxe_ByteGear/Materials/MF_Pickaxe_ByteGear_Glitch
graph {nodes [
  n1 = SetMaterialAttributes!(
   MaterialAttributes=NamedRerouteUsage()
   EmissiveColor=NamedRerouteUsage()
   Opacity=NamedRerouteUsage()
   BaseColor=NamedRerouteUsage()
   WorldPositionOffset=NamedRerouteUsage()
  ) {attributeSetTypes [
    #1
    #2
    #3
    #4
   ]}
  FunctionOutput(None=@n1) {outputName=MaterialAttributes, sortPriority=-1}
  FunctionOutput(None=NamedRerouteUsage()) {outputName=EmissiveColor, sortPriority=1}
  FunctionOutput(None=NamedRerouteUsage()) {outputName=Opacity}
  n2 = TextureCoordinate()
  Multiply(
   A=Panner(Coordinate=@n2) {speedX=5, speedY=10}
   B=ScalarParameter() {defaultValue=1, group=HolographicTexture_Glitch, parameterName=WPO_GlitchNoiseFrequency, sortPriority=33}
  )
  n3 = Panner(Coordinate=TextureCoordinate()) {speedX=5, speedY=10}
  n4 = ScalarParameter() {defaultValue=1, group=HolographicTexture_Glitch, parameterName=WPO_GlitchNoiseFrequency, sortPriority=33}
  n5 = Multiply(
   A=NamedRerouteUsage()
   B=TextureSample(UVs=Multiply(A=@n3, B=@n4)) {samplerType=SAMPLERTYPE_Masks, texture=/Game/Characters/Player/Male/Medium/Bodies/M_MED_Storm_Soldier/Textures/ElectricityNoise_M}.RGB
  )
  n6 = MaterialFunctionCall(Color=VertexNormalWS()) {materialFunction=&SplitComponents}
  n7 = Multiply(
   A=MaterialFunctionCall(R=@n6.R, G=@n6.R, B=Multiply(A=@n6.B) {constB=0.3}) {materialFunction=&AppendMany}.RGB
   B=ScalarParameter() {defaultValue=35, group=HolographicTexture_Glitch, parameterName=WPO_GlitchIntensity, sortPriority=33}
  ) {constB=35}
  NamedRerouteDeclaration(None=StaticSwitchParameter(True=Multiply(A=@n5, B=@n7) {constB=100}, False=NamedRerouteUsage()) {group=HolographicTexture_Glitch, parameterName=UseGlitchWPO, sortPriority=33}) {name=GlitchWPO, nodeColor=<rgba 1 0.9141 0 1>}
  n8 = Constant() {r=1}
  n9 = Subtract(A=@n8, B=ScalarParameter() {defaultValue=0.075, group=HolographicTexture_Glitch, parameterName=GlitchDuration, sortPriority=10})
  n10 = Multiply(A=Time(), B=ScalarParameter() {defaultValue=0.2, group=HolographicTexture_Glitch, parameterName=GlitchFrequency, sortPriority=10}) {constB=0.2}
  n11 = Add(A=Multiply(A=Sine(None=@n10)) {constB=0.5}) {constB=0.5}
  NamedRerouteDeclaration(None=Multiply(
    A=SmoothStep(Min=@n9, Max=@n8, Value=@n11) {constMax=0.9, constMin=0.8}
    B=MaterialFunctionCall() {materialFunction=/Game/MF__FloatFlickerRandom_Generic}.RandomFlickerFloat
   )) {name=GlitchFrequencyMask, nodeColor=<rgba 0.9377 1 0.8953 1>}
  n12 = MaterialFunctionCall!(
   ScanLine_Tiling_1
   Scanline_PannerSpeed_1
   ScanLine_Tiling
   Scanline_PannerSpeed
   Scanline_Rotation
   UVs=FunctionInput(Preview=TextureCoordinate()) {bUsePreviewValueAsDefault=true, inputName=Scanlines_UV, inputType=FunctionInput_Vector2, sortPriority=2}
  ) {desc=ScanLine, materialFunction=/Game/Global/Materials/MaterialGym/Materials/HolographicTexture/MF_HolographicTexture_ScanLines}
  n13 = Add(A=@n12.Result, B=ScalarParameter() {defaultValue=0.5, group=HolographicTexture, parameterName=Scanline_ColorMultiplier, sortPriority=0}) {constB=0}
  n14 = TextureCoordinate()
  n15 = TextureSampleParameter2D(UVs=MaterialFunctionCall!(Main_TextureUV=FunctionInput(Preview=@n14) {bUsePreviewValueAsDefault=true, inputName=TextureUV, inputType=FunctionInput_Vector2, sortPriority=2}) {materialFunction=/Game/Global/Materials/MaterialGym/Materials/ScreenGlitchGeneric/MF_ScreenGlitch_Generic}.Result) {
   group=HolographicTexture
   parameterName=HologramTexture
   samplerSource=SSM_Clamp_WorldGroupSettings
   sortPriority=-20
   texture=/Game/2dAssets/Sprays/Season25/Textures/T-Sprays-Season25-S25-GalaxyCupReward
  }
  n16 = Multiply(A=@n15.RGB, B=@n15.RGB)
  n17 = VectorParameter() {desc=HOLOGRAM BASE COLOR, group=HolographicTexture, parameterName=HologramBaseColor, sortPriority=-4, defaultValue=<rgba 0.1357 0.2884 1 1>}
  n18 = FunctionInput(Preview=@n17.RGB) {bUsePreviewValueAsDefault=true, inputName=HologramBaseColor}
  n19 = ScalarParameter() {defaultValue=1, group=HolographicTexture, parameterName=TextureBrightness, sortPriority=-19}
  n20 = StaticSwitchParameter(True=LinearInterpolate(A=@n18, B=Multiply(A=@n15.RGB, B=@n19), Alpha=@n15.A), False=@n18) {defaultValue=true, group=HolographicTexture, parameterName=UseHologramTexture, sortPriority=-5}
  n21 = ChannelMaskParameter(None=NamedRerouteUsage()) {group=HolographicTexture, maskChannel=Alpha, parameterName=TextureMask_Channel, sortPriority=-10, defaultValue=<rgba 0 0 0 1>}
  n22 = TextureSampleParameter2D() {
   group=HolographicTexture
   parameterName=HologramMask
   sortPriority=-10
   texture=/BRCosmetics/Weapons/FORT_Melee/Pickaxe_CosmicSquatter/Textures/T_Pickaxe_CosmicSquatter_FX
  }
  n23 = ChannelMaskParameter(None=@n22.RGB) {group=HolographicTexture, parameterName=HologramMask_Channel, sortPriority=-10}
  n24 = StaticSwitchParameter(
   True=@n21.R
   False=StaticSwitchParameter(
    True=FunctionInput(Preview=@n23.R) {bUsePreviewValueAsDefault=true, inputName=HologramMask, inputType=FunctionInput_Scalar, sortPriority=5}
    False=Constant() {r=1}
   ) {group=HolographicTexture, parameterName=UseHologramMask, sortPriority=-10}
  ) {group=HolographicTexture, parameterName=UseTextureOnly, sortPriority=-5}
  n25 = LinearInterpolate(
   A=NamedRerouteUsage()
   B=Multiply(
    A=Saturate(None=@n13)
    B=StaticSwitchParameter(True=@n16, False=@n20) {group=HolographicTexture, parameterName=UseTextureOnly, sortPriority=-5}
   )
   Alpha=@n24
  )
  n26 = NamedRerouteUsage()
  n27 = Power(Base=@n26) {constExponent=4}
  NamedRerouteDeclaration(None=LinearInterpolate(A=Add(A=@n25, B=@n26), B=Multiply(A=@n25, B=@n27), Alpha=NamedRerouteUsage())) {name=EmissiveColor, nodeColor=<rgba 1 0 0.0469 1>}
  NamedRerouteDeclaration(None=Multiply(
    A=NamedRerouteUsage()
    B=TextureSample(UVs=@n2) {samplerType=SAMPLERTYPE_LinearColor, texture=/BRCosmetics/Weapons/FORT_Melee/Pickaxe_ByteGear/Textures/T_ByteGear_Pickaxe_FX}.R
   )) {name=EmissiveCubeOutlines, nodeColor=<rgba 0 0.3672 1 1>}
  NamedRerouteDeclaration(None=@n15.RGBA) {name=Texture RGBA, nodeColor=<rgba 1 0 0.8906 1>}
  n28 = GetMaterialAttributes(None=FunctionInput() {inputName=MaterialAttributes, inputType=FunctionInput_MaterialAttributes, sortPriority=-1}) {attributeGetTypes [
    #3
    #4
    #1
    #2
   ]}
  NamedRerouteDeclaration(None=@n28.BaseColor) {name=OG Base Color, nodeColor=<rgba 0 0.7422 1 1>}
  NamedRerouteDeclaration(None=@n28.WorldPositionOffset) {name=OG WPO, nodeColor=<rgba 0 1 0.6016 1>}
  NamedRerouteDeclaration(None=@n28.EmissiveColor) {name=OG Emissive, nodeColor=<rgba 0 1 0.1797 1>}
  NamedRerouteDeclaration(None=@n28.Opacity) {name=OG Opacity, nodeColor=<rgba 0.3422 0.8799 1 1>}
  NamedRerouteDeclaration(None=@n28.MaterialAttributes) {name=OG Material Attributes, nodeColor=<rgba 0.8984 0 1 1>}
  n29 = NamedRerouteUsage()
  n30 = Constant()
  NamedRerouteDeclaration(None=StaticSwitchParameter(True=LinearInterpolate(A=@n29, B=@n30, Alpha=@n24), False=@n29) {defaultValue=true, group=HolographicTexture, parameterName=OverrideBaseColorOnHologram, sortPriority=-100}) {name=BaseColor, nodeColor=<rgba 0.5938 0 1 1>}
  n31 = ScalarParameter() {defaultValue=1, group=HolographicTexture, parameterName=HologramOpacity, sortPriority=-10}
  n32 = Saturate(None=Add(A=@n12.Result, B=ScalarParameter() {defaultValue=0.8, group=HolographicTexture, parameterName=Scanline_Opacity, sortPriority=0}) {constB=0.8})
  NamedRerouteDeclaration(None=LinearInterpolate(A=NamedRerouteUsage(), B=Multiply(A=@n31, B=@n32), Alpha=@n24)) {name=Opacity, nodeColor=<rgba 0 1 0.2734 1>}
 ]}
