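# Material function for the ByteGear pickaxe that layers a faked gemstone look onto incoming material attributes.
# Its output overrides opacity, specular, subsurface color and exposure-compensated emissive.
# Gem lighting combines refracted distance-corrected cubemap rays, sky-light reflection and refraction, custom GGX sun highlights and phase-function scattering.
# The graph also holds hologram glitch, scanline, noise offset and de-rez branches, with no direct link to the output.
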MaterialFunction /BRCosmetics/Weapons/FORT_Melee/Pickaxe_ByteGear/Materials/MF_Pickaxe_ByteGearB
graph {nodes [
  n1 = ComponentMask(None=FunctionInput() {inputName=TechArtMask, sortPriority=1}) {g=true}
  n2 = OneMinus(None=@n1)
  n3 = NamedRerouteUsage()
  BlendMaterialAttributes(
   A=SetMaterialAttributes!(MaterialAttributes=NamedRerouteUsage(), Opacity=Saturate(None=@n2)) {attributeSetTypes [#1]}
   B=@n3
   Alpha=@n1
  )
  Constant3Vector() {constant=<rgba 0 0 1 0>}
  Constant3Vector()
  Constant() {r=3}
  n4 = ShadingPathSwitch(Default=Constant(), Mobile=NamedRerouteUsage())
  n5 = NamedRerouteUsage()
  n6 = SetMaterialAttributes!(
   MaterialAttributes=FunctionInput() {bUsePreviewValueAsDefault=true, inputName=IN_MaterialAttributes, inputType=FunctionInput_MaterialAttributes}
   EmissiveColor=Divide(A=Add(A=@n4, B=@n5) {constB=0}, B=EyeAdaptation().EyeAdaptation)
   Opacity=ScalarParameter() {defaultValue=0.7, group=Gemstone, parameterName=GS_Opacity}
   SubsurfaceColor=Multiply(A=NamedRerouteUsage(), B=ScalarParameter() {parameterName=GS_SubsurfaceStrength})
   Specular=ScalarParameter() {defaultValue=1, group=Gemstone, parameterName=GS_Specular}
  ) {attributeSetTypes [
    #2
    #1
    #3
    #4
   ]}
  FunctionOutput(None=@n6)
  LinearInterpolate(A=NamedRerouteUsage(), B=ScalarParameter() {defaultValue=0.5, parameterName=B}, Alpha=NamedRerouteUsage()) {constA=1, constB=0}
  n7 = VectorParameter() {group=00_ExampleParameters, parameterName=GradientColor01, defaultValue=<rgba 1 0 0.6293 1>}
  n8 = VectorParameter() {group=00_ExampleParameters, parameterName=GradientColor02, defaultValue=<rgba 0 0.3224 1 1>}
  n9 = ComponentMask(None=NamedRerouteUsage()) {g=true}
  n10 = VectorParameter() {desc=HOLOGRAM BASE COLOR, group=HolographicTexture, parameterName=HologramBaseColor, sortPriority=-4, defaultValue=<rgba 0.286 1 0.1357 1>}
  n11 = WorldPosition()
  n12 = ScalarParameter() {defaultValue=0.005, group=00_ExampleParameters, parameterName=ScanlineTiling}
  n13 = NamedRerouteUsage()
  n14 = Add(
   A=MaterialFunctionCall(UVs=@n13, Texture Scale=ScalarParameter() {defaultValue=1, group=00_ExampleParameters, parameterName=TextureScale, sortPriority=1}) {materialFunction=&ScaleUVsByCenter}.UVs
   B=AppendVector(
    A=ScalarParameter() {group=00_ExampleParameters, parameterName=TexturePositionOffset_X, sortPriority=1}
    B=ScalarParameter() {group=00_ExampleParameters, parameterName=TexturePositionOffset_Y, sortPriority=1}
   )
  )
  n15 = ScalarParameter() {group=00_ExampleParameters, parameterName=TextureRotation, sortPriority=1}
  MaterialFunctionCall!(
   MaterialAttributes=@n3
   HologramBaseColor=StaticSwitchParameter(True=LinearInterpolate(A=@n7.RGB, B=@n8.RGB, Alpha=@n9), False=@n10.RGB) {group=00_ExampleParameters, parameterName=IsPlaneExample}
   Scanlines_UV=StaticSwitchParameter(True=Multiply(A=@n11.Z, B=@n12) {constB=0.005}, False=NamedRerouteUsage()) {group=00_ExampleParameters, parameterName=UseWorldPositionForScanlines}
   TextureUV=StaticSwitchParameter(True=Rotator(Coordinate=@n14, Time=@n15), False=@n13) {defaultValue=true, group=00_ExampleParameters, parameterName="OffsetTexturePosition&&Scale&&Rotation"}
   HologramMask
  ) {materialFunction=&/MF_Pickaxe_ByteGear_Glitch}
  n16 = GetMaterialAttributes(None=SetMaterialAttributes!(MaterialAttributes=NamedRerouteUsage(), EmissiveColor=NamedRerouteUsage(), Opacity=NamedRerouteUsage()) {attributeSetTypes [#2, #1]}) {attributeGetTypes [#1, #2]}
  n17 = SetMaterialAttributes!(MaterialAttributes=@n16.MaterialAttributes, Opacity=NamedRerouteUsage(), EmissiveColor=NamedRerouteUsage()) {attributeSetTypes [#1, #2]}
  MaterialFunctionCall(In=@n17) {materialFunction=&MF_TechArtOverrides_accessories}
  n18 = Multiply(A=SkyLightEnvMapSample(Direction=NamedRerouteUsage(), Roughness=ScalarParameter() {defaultValue=0.2, group=Gemstone, parameterName=GS_ReflectRoughness})) {constB=0.75}
  n19 = SkyLightEnvMapSample(Direction=NamedRerouteUsage(), Roughness=ScalarParameter() {defaultValue=0.2, group=Gemstone, parameterName=GS_RefractRoughness})
  n20 = Multiply(A=Multiply(A=@n19) {constB=1.2}, B=NamedRerouteUsage())
  n21 = NamedRerouteUsage()
  n22 = Saturate(None=Multiply(
    A=LinearInterpolate(A=@n18, B=@n20, Alpha=@n21)
    B=ScalarParameter() {defaultValue=1, group=Gemstone, parameterName=GS_ReflectionsBrightness}
   ))
  n23 = ScalarParameter() {defaultValue=1, group=Gemstone, parameterName=GS_RefractionsExponent}
  n24 = Multiply(
   A=NamedRerouteUsage()
   B=Desaturation(None=Power(Base=@n22, Exp=@n23), Fraction=ScalarParameter() {group=Gemstone, parameterName=GS_RefractionsDesaturation})
  )
  Multiply(A=NamedRerouteUsage(), B=@n24) {constA=1}
  n25 = MaterialFunctionCall!(
   CubePosition=Add(A=ActorPositionWS(), B=VectorParameter() {group=Gemstone, parameterName=GS_CubeMapPositionOffset, defaultValue=<rgba 0 0 0 1>}.RGB)
   Radius=ScalarParameter() {defaultValue=150, group=Gemstone, parameterName=GS_RefractedCubeMapRadius}
   UseRefractedRay=StaticBool() {value=true}
   IOR=ScalarParameter() {defaultValue=1, parameterName=GS_IncomingIOR}
   CustomNormalWS=Transform(None=TextureSampleParameter2D() {group=Gemstone, parameterName=GS_NormalMap, samplerType=SAMPLERTYPE_Normal, texture=/Engine/EngineMaterials/FlatNormal}.RGB)
  ) {materialFunction=/Game/Global/Materials/MaterialGym/Materials/Gem/MaterialFunctions/MF_DistanceCorrectedCubemapSample}
  n26 = TextureSampleParameterCube!(UVs=NamedRerouteDeclaration(None=@n25.DistanceCorrectedRay) {name=UVW, nodeColor=<rgba 0 0.625 1 1>}, Apply View MipBias) {
   group=Gemstone
   parameterName=GS_RefractionCubeMap
   samplerType=SAMPLERTYPE_LinearColor
   texture=/Game/Global/Materials/MaterialGym/Materials/Gem/Textures/TC_InteriorRoomCube
  }
  Multiply(A=@n26.RGB, B=NamedRerouteDeclaration(None=@n25.Mask) {name=CubeMask, nodeColor=<rgba 0 1 0.6016 1>})
  n27 = ComponentMask(None=TransformPosition(None=WorldPosition().XYZ) {transformSourceType=TRANSFORMPOSSOURCE_World}) {b=true}
  n28 = ComponentMask(None=TextureCoordinate()) {g=true}
  n29 = ScalarParameter() {parameterName=MaskPoison}
  n30 = AppendVector(
   A=ComponentMask(None=TextureCoordinate() {vTiling=0.5}) {r=true}
   B=Divide(A=LinearInterpolate(A=@n27, B=@n28, Alpha=@n29), B=ScalarParameter() {defaultValue=128, parameterName=FlickerScale}) {constB=32}
  )
  n31 = Ceil(None=Multiply(A=Time()) {constB=3})
  n32 = VectorParameter() {parameterName=Param_1, defaultValue=<rgba 0.1 -0.1 0 1>}
  n33 = Divide(A=@n32.R) {constB=1000}
  n34 = TextureSample(UVs=Panner(Coordinate=@n30, Time=@n31, Speed=AppendVector(A=@n32.R, B=@n33))) {samplerType=SAMPLERTYPE_LinearColor, texture=/Game/UI/Foundation/Textures/Pattern/T_CubePattern}
  n35 = VectorParameter() {defaultValue=<rgba 0.3 -0.1 0 1>}
  n36 = Divide(A=@n35.G) {constB=1000}
  n37 = TextureSample(UVs=Panner(Coordinate=@n30, Time=@n31, Speed=AppendVector(A=@n36, B=@n35.G))) {samplerType=SAMPLERTYPE_LinearColor, texture=&T_CubePattern}
  n38 = OneMinus(None=Round(None=Subtract(A=@n34.R, B=@n37.G)))
  NamedRerouteDeclaration(None=LinearInterpolate(A=Constant() {r=1}, B=Clamp(None=@n38), Alpha=ScalarParameter() {parameterName=Flicker})) {name=ScanLinesVerB, nodeColor=<rgba 0 1 0.5781 1>}
  n39 = GetMaterialAttributes(None=FunctionInput() {inputType=FunctionInput_MaterialAttributes}) {attributeGetTypes [#5, #6]}
  NamedRerouteDeclaration(None=@n39.MaterialAttributes) {name=IN_BaseMaterial, nodeColor=<rgba 1 0.5859 0 1>}
  n40 = VertexColor()
  n41 = Multiply(A=Time(), B=ScalarParameter() {defaultValue=0.05, parameterName=WPO_Speed})
  n42 = TextureSample(UVs=Divide(A=Add(A=@n40.R, B=@n41), B=Constant2Vector() {g=1, r=64})) {samplerType=SAMPLERTYPE_LinearColor, texture=/Game/Effects/Fort_Effects/Textures/Tile/LinerNoise64x64}
  NamedRerouteDeclaration(None=@n42.G) {name=NoiseMapG, nodeColor=<rgba 0.1207 1 0.1311 1>}
  NamedRerouteDeclaration(None=@n42.B) {name=NoiseMapB, nodeColor=<rgba 0 0.2266 1 1>}
  n43 = VectorParameter() {parameterName=UV2, defaultValue=<rgba 0.5 1.5 0.1 0.206>}
  n44 = ComponentMask(None=@n43.RGB) {g=true, r=true}
  n45 = LinearInterpolate(Alpha=VertexColor().G) {constA=1, constB=0.309}
  n46 = Panner(Coordinate=Multiply(A=TextureCoordinate() {coordinateIndex=1}, B=Multiply(A=@n44, B=@n45)), Speed=AppendVector(A=@n43.B, B=@n43.B))
  n47 = Multiply(A=ComponentMask(None=NamedRerouteUsage()) {g=true, r=true}, B=ScalarParameter() {defaultValue=0.02, parameterName=NormalAffectMap})
  NamedRerouteDeclaration(None=TextureSample(UVs=Add(A=@n46, B=@n47)) {texture=&T_NoisePack_01}.R) {name=NoisePackMap_Red, nodeColor=<rgba 1 0.1389 0.1198 1>}
  n48 = TextureCoordinate() {coordinateIndex=1}
  n49 = ComponentMask(None=@n48) {r=true}
  n50 = AppendVector(A=Multiply(A=@n49, B=@n49), B=ComponentMask(None=@n48) {g=true})
  n51 = TextureSample(UVs=Panner(Coordinate=MaterialFunctionCall(XY=@n50) {materialFunction=&Swizzle}.YX) {speedX=1}) {texture=/Game/Effects/Fort_Effects/Textures/T_FX_Circuitboard2}
  NamedRerouteDeclaration(None=@n51.G) {name=CircuitBoardG, nodeColor=<rgba 1 0 0.8672 1>}
  NamedRerouteDeclaration(None=@n51.B) {name=CircuitBoardB, nodeColor=<rgba 0 1 0.8125 1>}
  n52 = Time()
  n53 = Multiply(A=VertexColor().R) {constB=2}
  n54 = Sine(None=Multiply(A=Add(A=@n52, B=@n53), B=ScalarParameter() {defaultValue=0.15, parameterName=WiggleFreq}))
  n55 = ScalarParameter() {defaultValue=2, parameterName=WiggleDistance}
  n56 = Multiply(A=Transform(None=Constant3Vector() {constant=<rgba 0 0 1 0>}) {transformSourceType=TRANSFORMSOURCE_Local}, B=@n55)
  NamedRerouteDeclaration(None=Clamp(None=Multiply(A=@n54, B=@n56), Max=@n55) {maxDefault=5}) {name=WiggleAmount, nodeColor=<rgba 1 0.4922 0 1>}
  n57 = Power(Base=NamedRerouteUsage(), Exp=ScalarParameter() {defaultValue=4, parameterName=WPO_MotionExp})
  n58 = Subtract(
   A=MaterialFunctionCall(X=ComponentMask(None=@n57) {r=true}, Y=NamedRerouteUsage(), Z=NamedRerouteUsage()) {materialFunction=&MakeFloat3}.Result
   B=Constant3Vector() {constant=<rgba 0 0.5 0.5 0>}
  )
  n59 = VectorParameter() {parameterName=WPO_Rand_Size, defaultValue=<rgba 6 0.5 0.5 1.25>}
  NamedRerouteDeclaration(None=Transform(None=Multiply(A=@n58, B=@n59.RGB)) {transformSourceType=TRANSFORMSOURCE_Local}) {name=WorldSpaceNoise, nodeColor=<rgba 0.1719 0 1 1>}
  n60 = NamedRerouteUsage()
  n61 = VectorParameter() {parameterName=WPO_Rand_Size, defaultValue=<rgba 6 0.5 0.5 1.25>}
  n62 = NamedRerouteUsage()
  n63 = WorldPosition()
  n64 = Normalize(VectorInput=ComponentMask(None=TransformPosition(None=@n63.XYZ) {transformSourceType=TRANSFORMPOSSOURCE_World}) {r=true})
  n65 = Saturate(None=If(A=VertexColor().G, "A > B"=Multiply(A=@n57, B=@n61.A) {constB=2}, "A < B"=Multiply(A=@n62, B=@n64)) {constB=0.5})
  NamedRerouteDeclaration(None=SetMaterialAttributes!(MaterialAttributes, WorldPositionOffset=Add(A=@n60, B=@n65)) {attributeSetTypes [#7]}) {name=MaterialAttributeWPO, nodeColor=<rgba 0.1484 1 0 1>}
  n66 = TextureSampleParameter2D() {
   group=Emissive Mask
   parameterName=Emissive Mask Texture
   sortPriority=1
   texture=/BRCosmetics/Accessories/FORT_Backpacks/F_MED_Graffiti_Future_Backpack/Skins/Orange/Textures/T_F_MED_GraffitiFuture_Orange_backpack_FX
  }
  NamedRerouteDeclaration(None=@n66.G) {name=EmissiveMaskTextureG, nodeColor=<rgba 0.057 1 0.1229 1>}
  n67 = VectorParameter() {
   group=Emissive Mask
   parameterName=Emissive Mask Color/Alpha
   sortPriority=3
   channelNames {a=Alpha Strength, b="", g="", r=""}
   defaultValue=<rgba 0 0 0 5>
  }
  NamedRerouteDeclaration(None=@n67) {name=EmissiveMaskColorRGB, nodeColor=<rgba 0 0.7891 1 1>}
  NamedRerouteDeclaration(None=@n39.Normal) {name=BaseMaterial_Normal, nodeColor=<rgba 0.2656 0 1 1>}
  NamedRerouteDeclaration(None=@n67.Alpha Strength) {name=EmissiveMaskAlpha, nodeColor=<rgba 1 0.8906 0 1>}
  NamedRerouteDeclaration(None=ScalarParameter() {group=Emissive Mask, parameterName=Use Emissive Mask, sortPriority=0}) {name=UseEmissiveMaskAmount, nodeColor=<rgba 0.5703 0 1 1>}
  n68 = CameraVectorWS()
  n69 = Transform(None=LinearInterpolate(
    A=Constant3Vector() {constant=<rgba 0 0 1 0>}
    B=NamedRerouteUsage()
    Alpha=ScalarParameter() {defaultValue=0.25, parameterName=NormalStrength}
   ))
  n70 = CurveAtlasRowParameter(CurveTime=Saturate(None=DotProduct(A=@n68, B=@n69))) {
   atlas=&AthenaCurveAtlas01
   curve=/Game/Accessories/FORT_Backpacks/M_MED_StreetDemon_Backpack/FX/Curve_Color
   defaultValue=81
   parameterName=ColorCurveFromNormalMap
  }
  NamedRerouteDeclaration(None=MaterialFunctionCall(Hue Shift Percentage=ScalarParameter() {parameterName=HueShiftColorCurve}, Texture=@n70) {materialFunction=&HueShift}.Result) {name=ColorCurveFromNormalMapRGB, nodeColor=<rgba 0 1 0.25 1>}
  n71 = ComponentMask(None=TextureCoordinate() {coordinateIndex=1}) {g=true}
  n72 = CurveAtlasRowParameter(CurveTime=Power(Base=OneMinus(None=@n71))) {
   atlas=&AthenaCurveAtlas01
   curve=/Game/Athena/LinearColorCurves/HDRFlamesCurve
   defaultValue=60
   parameterName=ColorCurveHDRFlamesCurve
  }
  NamedRerouteDeclaration(None=@n72) {name=ColorCurveHDRFlamesCurve, nodeColor=<rgba 1 0 0.0703 1>}
  NamedRerouteDeclaration(None=@n70.A) {name=ColorCurveFromNormalMapA, nodeColor=<rgba 0 0.3906 1 1>}
  n73 = NamedRerouteUsage()
  n74 = LinearInterpolate(
   A=NamedRerouteUsage()
   B=LinearInterpolate(A=NamedRerouteUsage(), B=NamedRerouteUsage(), Alpha=NamedRerouteUsage())
   Alpha=NamedRerouteUsage()
  )
  NamedRerouteDeclaration(None=LinearInterpolate(A=NamedRerouteUsage(), B=Multiply(A=@n73, B=@n74), Alpha=ScalarParameter() {parameterName=Use_WorldPositionMask})) {name=EmissiveColorCalculation, nodeColor=<rgba 0.7109 1 0 1>}
  n75 = LinearInterpolate(
   A=NamedRerouteUsage()
   B=LinearInterpolate(A=NamedRerouteUsage(), B=NamedRerouteUsage(), Alpha=Constant() {r=0.3})
   Alpha=ScalarParameter() {parameterName=MaskPoison}
  )
  NamedRerouteDeclaration(None=Multiply(A=NamedRerouteUsage(), B=LinearInterpolate(Alpha=@n75) {constA=0.25})) {name=OpacityCalculation, nodeColor=<rgba 0.9688 0 1 1>}
  NamedRerouteDeclaration(None=@n16.Opacity) {name=UpdatedOpacity, nodeColor=<rgba 0 1 0.6484 1>}
  n76 = Multiply(A=Multiply(A=NamedRerouteUsage(), B=NamedRerouteUsage()), B=NamedRerouteUsage())
  n77 = NamedRerouteUsage()
  n78 = NamedRerouteUsage()
  NamedRerouteDeclaration(None=LinearInterpolate(A=@n76, B=LinearInterpolate(A=@n76, B=@n77, Alpha=@n78), Alpha=NamedRerouteUsage())) {name=OpacityWithScanLines, nodeColor=<rgba 1 0.3281 0 1>}
  NamedRerouteDeclaration(None=@n16.EmissiveColor) {name=UpdatedEmissiveColor, nodeColor=<rgba 0.8794 1 0.0276 1>}
  NamedRerouteDeclaration(None=@n39.BaseColor) {name=BaseMaterial_Color, nodeColor=<rgba 0.0547 1 0 1>}
  n79 = Multiply(A=NamedRerouteUsage(), B=NamedRerouteUsage())
  n80 = NamedRerouteUsage()
  n81 = NamedRerouteUsage()
  NamedRerouteDeclaration(None=LinearInterpolate(A=@n79, B=LinearInterpolate(A=@n79, B=@n80, Alpha=@n81), Alpha=NamedRerouteUsage())) {name=EmissiveColorWithScanlines, nodeColor=<rgba 0.3125 1 0 1>}
  NamedRerouteDeclaration(None=MaterialFunctionCall!(Fresnel=ScalarParameter() {defaultValue=1, parameterName=Fresnel}, Material=@n17) {materialFunction=/Game/Packages/Fortress_SharedMaterials/Base_Material_Functions/CharacterDeRez_Transition}.Result) {name=FinalMaterialAttributes, nodeColor=<rgba 0.4072 0.0343 0.0056 1>}
  n82 = NamedRerouteUsage()
  n83 = ConstantBiasScale(None=@n26.RGB) {bias=-0.5, scale=2}
  n84 = NamedRerouteDeclaration(None=Multiply(A=Normalize(VectorInput=@n83)) {constB=-1}) {name=InteriorNormal, nodeColor=<rgba 0.5234 1 0 1>}
  n85 = NamedRerouteUsage()
  n86 = Constant() {r=1}
  NamedRerouteDeclaration(None=Multiply(A=MaterialFunctionCall!(Ray Direction=@n82, Surface Normal=@n84, Refractive Index Origin=@n85, Refractive Index Target=@n86) {materialFunction=/Engine/Functions/Engine_MaterialFunctions01/Vectors/Refract}.Vector) {constB=-1}) {name=RefractVectorInterior, nodeColor=<rgba 1 0.9375 0 1>}
  n87 = Multiply(A=DotProduct(A=NamedRerouteUsage(), B=NamedRerouteUsage())) {constB=-1}
  n88 = ScalarParameter() {defaultValue=0.73}
  n89 = ScalarParameter() {defaultValue=0.78, parameterName=Param_1}
  NamedRerouteDeclaration(None=LinearInterpolate(
    A=ScalarParameter() {defaultValue=0.5, group=Gemstone, parameterName=GS_ReflectionAngleMin}
    Alpha=MaterialFunctionCall(Input=@n87, Input Low=@n88, Input High=@n89) {materialFunction=&RemapValue_0-1}.Result
   ) {constA=0.5}) {name=TIR, nodeColor=<rgba 0 0.7422 1 1>}
  NamedRerouteDeclaration(None=ReflectionVectorWS(CustomWorldNormal=NamedRerouteUsage())) {name=ReflectionVector, nodeColor=<rgba 1 0 0.3281 1>}
  n90 = Constant3Vector() {constant=<rgba 0 0 100 0>}
  NamedRerouteDeclaration(None=Multiply(A=SkyAtmosphereLightIlluminance!(World Position=@n90)) {constB=0.5}) {name=DirectionalLightColor, nodeColor=<rgba 0 0.4844 1 1>}
  n91 = VectorParameter() {group=Gemstone, parameterName=InColor, defaultValue=<rgba 0 0 0 1>}
  n92 = VectorParameter() {group=Gemstone, parameterName=GS_ColorTint, sortPriority=-2, defaultValue=<rgba 0.0272 0.5312 0.5101 1>}
  n93 = ScalarParameter() {defaultValue=1, group=Gemstone, parameterName=GS_BaseColorBlend, sortPriority=-1}
  NamedRerouteDeclaration(None=Desaturation(None=LinearInterpolate(A=@n91.RGB, B=@n92.RGB, Alpha=@n93), Fraction=Constant() {r=-0.2})) {name=Tint, nodeColor=<rgba 0.3125 1 0 1>}
  n94 = NamedRerouteUsage()
  n95 = Constant() {r=-1}
  n96 = Constant() {r=1}
  NamedRerouteDeclaration(None=LinearInterpolate(
    A=ScalarParameter() {defaultValue=0.5, group=Gemstone, parameterName=GS_GradientLow}
    B=ScalarParameter() {defaultValue=1, group=Gemstone, parameterName=GS_GradientHigh}
    Alpha=MaterialFunctionCall(Input=@n94, Input Low=@n95, Input High=@n96) {materialFunction=&RemapValue_0-1}.Result
   ) {constA=0.45}) {name=Depth Gradient, nodeColor=<rgba 1 0 0.2812 1>}
  n97 = MaterialFunctionCall() {materialFunction=&BoundingBoxBased_0-1_UVW}
  NamedRerouteDeclaration(None=Multiply(A=ComponentMask(None=@n97.RGB) {b=true, r=true})) {name=ScanlineUVs, nodeColor=<rgba 0 1 0.8125 1>}
  MaterialExpressionNormalize_2 = Normalize(VectorInput=SkyAtmosphereLightDirection())
  MaterialExpressionCameraVectorWS_1 = CameraVectorWS()
  MaterialExpressionNamedRerouteUsage_49 = NamedRerouteUsage()
  MaterialExpressionConstant_28 = Constant() {r=0.3}
  MaterialExpressionConstant3Vector_3 = Constant3Vector() {constant=<rgba 1 1 1 0>}
  n98 = Custom!(
   L=@MaterialExpressionNormalize_2
   V=@MaterialExpressionCameraVectorWS_1
   N=@MaterialExpressionNamedRerouteUsage_49
   roughness=@MaterialExpressionConstant_28
   F0=@MaterialExpressionConstant3Vector_3
  ) {
   code="    float3 H = normalize(L + V);\r\n    float NdotH = saturate(dot(N, H));\r\n    float NdotL = saturate(dot(N, L));\r\n    float NdotV = saturate(dot(N, V));\r\n    float3 F = F0 + (1.0 - F0) * pow(1.0 - NdotV, 5.0);\r\n    float epsilon = .0001;\r\n\r\n\tfloat alpha = roughness * roughness;\r\n    float alphaSqr = alpha * alpha;\r\n\r\n    float denominator = NdotH * NdotH * (alphaSqr - 1.0) + 1.0;\r\n    float D = alphaSqr / (PI * denominator * denominator + epsilon);\r\n\t\r\n\t\r\n\t\r\n    float k = alpha / 2;    \r\n    float g_v = NdotV / ( NdotV*(1 - k ) + k);\r\n    float g_l = NdotL / (NdotL*(1 - k ) + k);\r\n    float G = g_v * g_l;\r\n    return NdotL*(D * G * F) / (4.0 * NdotL * NdotV + epsilon );"
   description=GGX
   inputs [
    {
     inputName=L
     input {expression=&self:MaterialExpressionNormalize_2}
    }
    {
     inputName=V
     input {expression=&self:MaterialExpressionCameraVectorWS_1}
    }
    {
     inputName=N
     input {expression=&self:MaterialExpressionNamedRerouteUsage_49}
    }
    {
     inputName=roughness
     input {expression=&self:MaterialExpressionConstant_28}
    }
    {
     inputName=F0
     input {
      expression=&self:MaterialExpressionConstant3Vector_3
      mask=1
      maskB=1
      maskG=1
      maskR=1
     }
    }
   ]
  }
  n99 = Multiply(A=@n98) {constA=1}
  n100 = DotProduct(A=NamedRerouteUsage(), B=@MaterialExpressionNormalize_2)
  n101 = Multiply(A=Power(Base=Saturate(None=@n100)) {constExponent=24}) {constB=5}
  n102 = NamedRerouteUsage()
  n103 = Multiply(A=Clamp(None=ShadingPathSwitch(Default=@n99, Mobile=@n101)) {maxDefault=40}, B=Multiply(A=@n102, B=@n102))
  n104 = NamedRerouteUsage()
  n105 = NamedRerouteUsage()
  NamedRerouteDeclaration(None=Multiply(A=Multiply(A=@n103, B=@n104), B=LinearInterpolate(Alpha=@n105) {constA=0.5, constB=0.05})) {name=SunInternalReflections, nodeColor=<rgba 0 0.8828 1 1>}
  n106 = DotProduct(A=SkyAtmosphereLightDirection(), B=NamedRerouteUsage())
  n107 = ScalarParameter() {defaultValue=-0.985, group=Gemstone, parameterName=GS_Phase}
  n108 = Clamp(None=Multiply(
    A=NamedRerouteUsage()
    B=MaterialFunctionCall!(Cosine=@n106, Anisotropy=@n107) {materialFunction=/Engine/Functions/Engine_MaterialFunctions01/Volumetrics/SchlickPhase}.Result
   )) {maxDefault=20}
  n109 = NamedRerouteUsage()
  n110 = NamedRerouteUsage()
  NamedRerouteDeclaration(None=Multiply(A=Multiply(A=@n108, B=@n109) {constB=2}, B=LinearInterpolate(Alpha=@n110) {constA=0.2})) {name=SunNscatter, nodeColor=<rgba 1 0.7031 0 1>}
  NamedRerouteDeclaration(None=@n24) {name=Sky Refractions, nodeColor=<rgba 0.6875 1 0 1>}
  n111 = Constant() {r=1}
  n112 = ComponentMask(None=NamedRerouteUsage()) {b=true}
  n113 = Saturate(None=OneMinus(None=Multiply(A=@n112) {constB=-1}))
  NamedRerouteDeclaration(None=MaterialFunctionCall!(1=@n111, 2=Constant() {r=0.5}, 3=@n111, 4, A=Multiply(A=@n113, B=@n113)) {materialFunction=/Engine/Functions/Engine_MaterialFunctions02/Lerp_Multiple_Float}.Lerp 3 Inputs) {name=GroundGradient, nodeColor=<rgba 0.4062 0 1 1>}
  NamedRerouteDeclaration(None=LinearInterpolate(
    A=Constant3Vector() {constant=<rgba 0 0 1 0>}
    B=Multiply(A=@n25.DistanceCorrectedRay, B=@n25.DistanceCorrectedRay)
    Alpha=ScalarParameter() {parameterName=Param_2}
   ) {constAlpha=0}) {name=DebugNormal, nodeColor=<rgba 0.6875 0 1 1>}
  NamedRerouteDeclaration(None=@n25.RefractedRay) {name=RefractedRay, nodeColor=<rgba 1 0 0.8906 1>}
  n114 = Multiply(A=Multiply(A=NamedRerouteUsage()) {constB=0.5}) {constB=0.5}
  n115 = NamedRerouteUsage()
  n116 = SkyAtmosphereLightDirection()
  n117 = LinearInterpolate(A=ScalarParameter() {defaultValue=0.5, parameterName=A}, Alpha=Abs(None=DotProduct(A=@n115, B=@n116))) {constA=0.25}
  NamedRerouteDeclaration(None=Multiply(A=Multiply(A=@n114, B=@n117), B=ScalarParameter() {defaultValue=1, group=Gemstone, parameterName=GS_SS_Strength})) {name=SubSurface, nodeColor=<rgba 1 0 0.5859 1>}
  n118 = NamedRerouteUsage()
  n119 = NamedRerouteUsage()
  NamedRerouteDeclaration(None=Add(A=Add(A=@n118, B=@n119), B=NamedRerouteUsage())) {name=Emissive, nodeColor=<rgba 1 0.9609 0 1>}
  NamedRerouteDeclaration(None=@n19) {name=debug, nodeColor=<rgba 0.1719 0 1 1>}
  n120 = WorldPosition()
  n121 = LinearInterpolate(
   A=ComponentMask(None=TransformPosition(None=@n120.XYZ) {transformSourceType=TRANSFORMPOSSOURCE_World}) {b=true}
   B=ComponentMask(None=TextureCoordinate()) {g=true}
   Alpha=ScalarParameter() {parameterName=MaskPoison}
  )
  n122 = VectorParameter() {parameterName=LineSpeed, defaultValue=<rgba 0.1 -0.1 0 0>}
  n123 = Panner(
   Coordinate=Divide(A=@n121, B=ScalarParameter() {defaultValue=12, parameterName=LineScale1}) {constB=32}
   Speed=AppendVector(A=@n122.R, B=Divide(A=@n122.R) {constB=1000})
  )
  NamedRerouteDeclaration(None=LinearInterpolate(Alpha=TextureSample(UVs=@n123) {samplerType=SAMPLERTYPE_Normal, texture=/Game/Global/Textures/Default/Noise/64x64Noise}.R) {constA=1, constAlpha=0.1, constB=0.55}) {name=AnimatedLinesA, nodeColor=<rgba 1 0.2578 0 1>}
  n124 = Panner(
   Coordinate=Divide(A=@n121, B=ScalarParameter() {defaultValue=128, parameterName=LineScale2}) {constB=32}
   Speed=AppendVector(A=@n122.G, B=Divide(A=@n122.G) {constB=10000000})
  )
  n125 = Multiply(A=Subtract(
    A=TextureSample(UVs=@n124) {samplerType=SAMPLERTYPE_Normal, texture=&64x64Noise}.G
    B=ScalarParameter() {defaultValue=0.9, parameterName=LineSub}
   ) {constB=0.99}) {constB=15}
  NamedRerouteDeclaration(None=LinearInterpolate(Alpha=Saturate(None=@n125)) {constA=0.5, constB=2}) {name=AnimatedLinesB, nodeColor=<rgba 0 0.0625 1 1>}
  NamedRerouteDeclaration(None=ScalarParameter() {defaultValue=1, group=Gemstone, parameterName=GS_IOR}) {name=IOR, nodeColor=<rgba 1 0 0.2109 1>}
  n126 = VectorParameter() {parameterName=UV, defaultValue=<rgba 0.5 0.5 0.05 0.0679>}
  n127 = Multiply(A=TextureCoordinate(), B=ComponentMask(None=@n126.RGB) {g=true, r=true})
  n128 = AppendVector(A=@n126.B, B=@n126.B)
  NamedRerouteDeclaration(None=TextureSampleParameter2D(UVs=Panner(Coordinate=@n127, Speed=@n128)) {parameterName=Param, samplerType=SAMPLERTYPE_Normal, texture=/Game/Global/Textures/Default/Tileable_Patterns/Cloud_normals}.RGB) {name=NormalMap, nodeColor=<rgba 0.7344 0 1 1>}
  NamedRerouteDeclaration(None=@n42.RGB) {name=NoiseMapRGB, nodeColor=<rgba 1 0.8883 0.8802 1>}
  NamedRerouteUsage()
  NamedRerouteUsage()
  ScalarParameter() {group=Gemstone, parameterName=GS_Metallic}
  ScalarParameter() {defaultValue=0.1, group=Gemstone, parameterName=GS_Roughness}
  ScalarParameter() {defaultValue=1.5, parameterName=TestIOR}
  SetMaterialAttributes!(
   MaterialAttributes
   BaseColor=NamedRerouteUsage()
   Roughness=Constant() {r=0.5}
   Metallic=Constant()
   Specular=Constant() {r=0.5}
   EmissiveColor=Constant()
   Normal=NamedRerouteUsage()
  ) {attributeSetTypes [
    #6
    #8
    #9
    #4
    #2
    #5
   ]}
  Step(X=Saturate(None=@n25.Mask)) {constX=0.5, constY=0.5}
  TextureCoordinate() {coordinateIndex=1}
  VectorParameter() {parameterName=GS_BaseColor, defaultValue=<rgba 0 0 0 1>}
 ]}
