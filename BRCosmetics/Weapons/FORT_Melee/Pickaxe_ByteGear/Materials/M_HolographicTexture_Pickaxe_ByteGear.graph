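# Translucent Default Lit hologram material for the ByteGear pickaxe, built on the shared MF_HolographicTexture_Generic function.
# Base color texture and constant surface attributes feed the function alongside hologram color, scanline UVs and texture UVs.
# Static switches select solid color or local-Y gradient, world-Z or local-XY scanlines, and raw or offset, scaled, rotated UV1.
# Local position is pre-skinning and passes through a vertex interpolator.
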Material /BRCosmetics/Weapons/FORT_Melee/Pickaxe_ByteGear/Materials/M_HolographicTexture_Pickaxe_ByteGear
graph {
 properties {
  bUseMaterialAttributes=true
  bUsedWithMeshDeformer=true
  bUsedWithSkeletalMesh=true
  blendMode=BLEND_Translucent
  pixelDepthOffsetMode=PDOM_Legacy
  usedShadingModels=Default Lit
 }
 nodes [
  Constant() {r=3}
  Reroute()
  n1 = TextureSampleParameter2D() {
   group=00_ExampleParameters
   parameterName=BaseColorTexture
   texture=/BRCosmetics/Accessories/FORT_Backpacks/Backpack_F_MED_ShinyStar/Textures/T_F_MED_ShinyStar_Backpack_D
  }
  n2 = SetMaterialAttributes!(
   MaterialAttributes
   Base Color=@n1.RGB
   Roughness=Constant() {r=0.5}
   Metallic=Constant()
   Specular=Constant() {r=0.5}
   Emissive Color=Constant()
   Normal=Constant3Vector() {constant=<rgba 0 0 1 0>}
  ) {attributeSetTypes [
    #1
    #2
    #3
    #4
    #5
    #6
   ]}
  n3 = Multiply(A=VertexInterpolator(VS=LocalPosition() {includedOffsets=ExcludeOffsets, localOrigin=InstancePreSkinning}.XY).PS) {constB=30}
  n4 = LinearInterpolate(
   A=VectorParameter() {group=00_ExampleParameters, parameterName=GradientColor01, defaultValue=<rgba 1 0 0.6293 1>}.RGB
   B=VectorParameter() {group=00_ExampleParameters, parameterName=GradientColor02, defaultValue=<rgba 0 0.3224 1 1>}.RGB
   Alpha=ComponentMask(None=@n3) {g=true}
  )
  n5 = VectorParameter() {desc=HOLOGRAM BASE COLOR, group=HolographicTexture, parameterName=HologramBaseColor, sortPriority=-4, defaultValue=<rgba 0.1357 0.2884 1 1>}
  n6 = Multiply(A=WorldPosition().Z, B=ScalarParameter() {defaultValue=0.005, group=00_ExampleParameters, parameterName=ScanlineTiling}) {constB=0.005}
  n7 = TextureCoordinate() {coordinateIndex=1}
  n8 = ScalarParameter() {defaultValue=1, group=00_ExampleParameters, parameterName=TextureScale, sortPriority=1}
  n9 = ScalarParameter() {group=00_ExampleParameters, parameterName=TexturePositionOffset_X, sortPriority=1}
  n10 = ScalarParameter() {group=00_ExampleParameters, parameterName=TexturePositionOffset_Y, sortPriority=1}
  n11 = Rotator(
   Coordinate=Add(
    A=MaterialFunctionCall(UVs=@n7, Texture Scale=@n8) {materialFunction=&ScaleUVsByCenter}.UVs
    B=AppendVector(A=@n9, B=@n10)
   )
   Time=ScalarParameter() {group=00_ExampleParameters, parameterName=TextureRotation, sortPriority=1}
  )
  SetMaterialAttributes!(
   MaterialAttributes=MaterialFunctionCall!(
    MaterialAttributes=@n2
    HologramBaseColor=StaticSwitchParameter(True=@n4, False=@n5.RGB) {group=00_ExampleParameters, parameterName=IsPlaneExample}
    Scanlines_UV=StaticSwitchParameter(True=@n6, False=@n3) {group=00_ExampleParameters, parameterName=UseWorldPositionForScanlines}
    TextureUV=StaticSwitchParameter(True=@n11, False=@n7) {group=00_ExampleParameters, parameterName="OffsetTexturePosition&&Scale&&Rotation"}
    HologramMask
   ) {materialFunction=/Game/Global/Materials/MaterialGym/Materials/HolographicTexture/MF_HolographicTexture_Generic}.MaterialAttributes
   Emissive Color
   Opacity
   Base Color
   World Position Offset
  ) {attributeSetTypes [
    #5
    #7
    #1
    #8
   ]}
  TextureCoordinate() {coordinateIndex=1}
  TextureCoordinate() {coordinateIndex=1}
 ]
}
