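# Animation Blueprint for the ByteGear pickaxe that adds physics-driven sway to its dyn_wire bone.
# Each update sets OnPawn from whether the pawn owner is valid.
# The input pose passes through two chained Anim Dynamics nodes toggled oppositely by that flag.
# On a pawn, a root-relative simulation with scaled component motion runs; otherwise, a world-space one with a softer angular spring takes over.
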AnimBlueprint /BRCosmetics/Weapons/FORT_Melee/Pickaxe_ByteGear/Meshes/ByteGear_Axe_AnimBP
target_skeleton=&/ByteGear_Axe_Skeleton
graphs [
 AnimGraph {nodes [
   AnimGraphNode_SubInput_0 = Input Pose()
   n1 = Get OnPawn() -> (OnPawn:bool)
   n2 = Anim Dynamics - Bone: dyn_wire!(ComponentPose=Local To Component(LocalPose=@AnimGraphNode_SubInput_0), bAlphaBoolEnabled:bool=@n1, GravityOverride:Vector="1,1,1") -> (Pose) {
    alphaInputType=Bool
    angularDampingOverride=1
    angularSpringConstant=1000
    bAngularSpring=true
    bGravityOverrideInSimSpace=true
    bOverrideAngularDamping=true
    bOverrideLinearDamping=true
    linearDampingOverride=1
    linearSpringConstant=5000
    simulationSpace=RootRelative
    alphaBoolBlend {blendOption=Linear}
    boundBone {boneName=dyn_wire}
    componentAppliedLinearAccClamp=<3000 3000 3000>
    componentLinearAccScale=<2 2 2>
    componentLinearVelScale=<1 1 1>
    physicsBodyDefinitions [{
      sphereCollisionRadius=0
      boundBone {boneName=dyn_wire}
      boxExtents=<6 4 6>
      constraintSetup {angularLimitsMax=<35 10 45>, angularLimitsMin=<-35 -10 -45>, angularTarget=<1 0 0>}
      localJointOffset=<0.2 0 0>
     }]
    relativeSpaceBone {boneName=b_AR_Root}
   }
   n3 = Anim Dynamics - Bone: dyn_wire(ComponentPose=@n2, bAlphaBoolEnabled=NOT Boolean(A=@n1), GravityOverride="1,1,1") {
    alphaInputType=Bool
    angularDampingOverride=1
    angularSpringConstant=600
    bAngularSpring=true
    bGravityOverrideInSimSpace=true
    bOverrideAngularDamping=true
    bOverrideLinearDamping=true
    linearDampingOverride=1
    linearSpringConstant=5000
    simulationSpace=World
    alphaBoolBlend {blendOption=Linear}
    boundBone {boneName=dyn_wire}
    componentAppliedLinearAccClamp=<6000 6000 6000>
    physicsBodyDefinitions [{
      sphereCollisionRadius=0
      boundBone {boneName=dyn_wire}
      boxExtents=<6 4 6>
      constraintSetup {angularLimitsMax=<35 10 45>, angularLimitsMin=<-35 -10 -45>, angularTarget=<1 0 0>}
      localJointOffset=<0 0 -0.001>
     }]
    relativeSpaceBone {boneName=b_AR_Root}
   }
   Output Pose(Result=Component To Local(ComponentPose=@n3)) {layerGroup=None}
  ]}
 EventGraph {nodes [
   Event BlueprintUpdateAnimation()
   > Set OnPawn(OnPawn:bool=IsValid(Object=TryGetPawnOwner())) -> (Output_Get:bool)
  ]}
]
variables [OnPawn {type=boolean}]
